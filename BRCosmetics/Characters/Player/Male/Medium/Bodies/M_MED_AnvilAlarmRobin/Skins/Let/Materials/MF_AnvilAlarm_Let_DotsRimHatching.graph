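# Material function for the AnvilAlarm "Let" skin that adds ShakeCrunch-style comic cel shading with halftone dots and hatching, switchable via UseShakeCrunchFX.
# Combines banded lighting, stylized specular and metallic, stepped-time paper detail and reactive glitch color, blending toward unlit-style emissive output.
# Procedural rotated dot and line patterns, thresholded by screen-projected vectors and lighting, form separate rim-light color and gradient outputs.
# Normals can blend toward vertex-color baked normals.
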MaterialFunction /BRCosmetics/Characters/Player/Male/Medium/Bodies/M_MED_AnvilAlarmRobin/Skins/Let/Materials/MF_AnvilAlarm_Let_DotsRimHatching
graph {nodes [
  n1 = NamedRerouteUsage()
  n2 = Saturate(None=Add(A=NamedRerouteUsage(), B=NamedRerouteUsage()))
  n3 = Constant()
  n4 = SetMaterialAttributes!(MaterialAttributes=@n1, EmissiveColor=@n2, BaseColor=@n2, Specular=@n3, Metallic=@n3, Roughness=Constant() {r=1}) {attributeSetTypes [
    #1
    #2
    #3
    #4
    #5
   ]}
  n5 = BlendMaterialAttributes(
   A=SetMaterialAttributes!(MaterialAttributes=@n1, BaseColor=@n2) {attributeSetTypes [#2]}
   B=@n4
   Alpha=ScalarParameter() {defaultValue=0.5, group=03_ShakeCrunch_DotsRimHatching, parameterName=CelShading_Blend, sliderMax=1, sortPriority=1}
  )
  n6 = NamedRerouteUsage()
  FunctionOutput(None=StaticSwitchParameter(
    True=SetMaterialAttributes!(MaterialAttributes=@n5, Normal=@n6) {attributeSetTypes [#6]}
    False=NamedRerouteUsage()
   ) {defaultValue=true, group=03_ShakeCrunch_DotsRimHatching, parameterName=UseShakeCrunchFX, sortPriority=0}) {outputName=MA_Out}
  n7 = NamedRerouteUsage()
  n8 = NamedRerouteUsage()
  n9 = NamedRerouteUsage()
  n10 = NamedRerouteUsage()
  FunctionOutput(None=Max(A=Max(A=@n7, B=@n8), B=Max(A=@n9, B=@n10))) {outputName=RimLight_Gradient, sortPriority=2}
  n11 = Power(Base=Desaturation(None=NamedRerouteUsage())) {constExponent=0.5}
  n12 = VectorParameter() {group=03_ShakeCrunch_DotsRimHatching, parameterName=GlitchColor_Primary, defaultValue=<rgba 1 0.0395 0.1119 1>}
  n13 = VectorParameter() {group=03_ShakeCrunch_DotsRimHatching, parameterName=GlitchColor_Secondary, defaultValue=<rgba 0 1 0.2582 1>}
  n14 = NamedRerouteUsage()
  NamedRerouteDeclaration(None=LinearInterpolate(
    A=MaterialFunctionCall(Base=@n11, Blend=@n12.RGB) {materialFunction=&Blend_Overlay}.Result
    B=MaterialFunctionCall(Base=@n11, Blend=@n13.RGB) {materialFunction=&Blend_Overlay}.Result
    Alpha=Saturate(None=@n14)
   )) {name=GlitchColor, nodeColor=<rgba 0.0312 0 1 1>}
  n15 = FunctionInput() {bUsePreviewValueAsDefault=true, inputName=FX, sortPriority=1}
  n16 = ChannelMaskParameter(None=@n15) {group=03_ShakeCrunch_DotsRimHatching, maskChannel=Blue, parameterName=FaceLine_Channel, defaultValue=<rgba 0 0 1 0>}
  NamedRerouteDeclaration(None=@n16.R) {name=FaceLine, nodeColor=<rgba 1 0 0.4219 1>}
  NamedRerouteDeclaration(None=ChannelMaskParameter(None=@n15) {group=03_ShakeCrunch_DotsRimHatching, parameterName=ColorVariation_Channel}.R) {name=ColorVariation, nodeColor=<rgba 0.175 1 0.4871 1>}
  n17 = MaterialFunctionCall() {materialFunction=/Game/Global/Materials/Character_Parent_2023/Anime_Parent_2023/MaterialFunctions/MFC_AnimeShader_LightDirection}
  n18 = ScalarParameter() {
   desc="LockedZ only for the specular highlight.\r\nfor a streched highlight in on the Z axis use the Blinn highlight"
   group=03_ShakeCrunch_DotsRimHatching
   parameterName=SpecularHighlight_LockedZVector
   sliderMax=1
  }
  n19 = AppendVector(A=Constant2Vector() {g=1, r=1}, B=Saturate(None=OneMinus(None=@n18)))
  n20 = Multiply(A=Add(A=CameraVectorWS(), B=@n17.Direction Vector), B=@n19)
  n21 = NamedRerouteUsage()
  n22 = Multiply(A=@n17.Direction Vector, B=@n19)
  n23 = StaticSwitchParameter(
   True=DotProduct(A=Normalize(VectorInput=@n20), B=@n21) {desc=Blinn}
   False=DotProduct(A=Normalize(VectorInput=@n22), B=ReflectionVectorWS(CustomWorldNormal=@n21)) {desc=Phong}
  ) {
   desc="Default is the Phong highlight.\r\nUse Blinn for a vertical streched highlight"
   group=03_ShakeCrunch_DotsRimHatching
   parameterName=UseBlinnHighlight
   sortPriority=0
  }
  n24 = NamedRerouteUsage()
  n25 = ScalarParameter() {group=03_ShakeCrunch_DotsRimHatching, parameterName=SpecularHighlight_CameraFacing_Blend, sliderMax=1}
  n26 = ScalarParameter() {defaultValue=0.1, group=03_ShakeCrunch_DotsRimHatching, parameterName=SpecularHighlight_Size, sliderMax=1, sliderMin=0.0001}
  n27 = SphereMask(
   A=Saturate(None=LinearInterpolate(A=@n23, B=@n24, Alpha=@n25))
   B=Constant() {r=1}
   Radius=@n26
   Hardness=ScalarParameter() {defaultValue=0.2, group=03_ShakeCrunch_DotsRimHatching, parameterName=SpecularHighlight_Hardness, sliderMax=1}
  ) {hardnessPercent=78}
  n28 = VectorParameter() {desc="RGB=Color, A=Opacity", group=03_ShakeCrunch_DotsRimHatching, parameterName=Specular_Color, defaultValue=<rgba 1 1 1 0>}
  n29 = Multiply(A=@n28.RGB, B=@n28.RGB)
  NamedRerouteDeclaration(None=Multiply(A=Multiply(A=@n27, B=@n29), B=NamedRerouteUsage())) {name=Specular_Curve, nodeColor=<rgba 1 0.4453 0 1>}
  n30 = VectorParameter() {group=03_ShakeCrunch_DotsRimHatching, parameterName=RimDotsVector, sortPriority=10, defaultValue=<rgba -1 0 0.5 1>}
  n31 = DotProduct(
   A=MaterialFunctionCall!(Vector=@n30.RGB) {materialFunction=/Game/Packages/Fortress_SharedMaterials/Anime_Master_Material/MF_ProjectToScreenSpace}.Result
   B=NamedRerouteUsage()
  )
  n32 = ScalarParameter() {defaultValue=-0.4, group=03_ShakeCrunch_DotsRimHatching, parameterName=RimDotsOffset, sliderMax=0.5, sliderMin=-0.5, sortPriority=20}
  n33 = ScalarParameter() {defaultValue=1.5, group=03_ShakeCrunch_DotsRimHatching, parameterName=RimDotsTransition, sliderMax=2, sortPriority=20}
  n34 = Subtract(A=Multiply(A=Add(A=@n31, B=@n32) {constB=-0.25}, B=@n33) {constB=1.5}, B=NamedRerouteUsage())
  n35 = Subtract(A=LinearInterpolate(A=@n34, B=Add(A=@n34) {constB=-0.25}, Alpha=NamedRerouteUsage()) {constA=1, constB=16}, B=NamedRerouteUsage())
  n36 = NamedRerouteDeclaration(None=SmoothStep(Min=Constant(), Max=Constant() {r=0.1}, Value=Saturate(None=@n35))) {name=RimDotsMask, nodeColor=<rgba 0.4766 1 0 1>}
  n37 = VectorParameter() {
   desc="RGB = Color\r\nA = Alpha / Instensity"
   group=03_ShakeCrunch_DotsRimHatching
   parameterName=RimDotsColor
   sortPriority=10
   defaultValue=<rgba 0.9 0.9 0.9 0.1>
  }
  n38 = NamedRerouteUsage()
  NamedRerouteDeclaration(None=AppendVector(A=Multiply(A=@n36, B=@n37.RGB), B=Multiply(A=@n37.A, B=@n38))) {name=RimDots, nodeColor=<rgba 1 0 0.6797 1>}
  n39 = Multiply(A=NamedRerouteUsage()) {constB=-1}
  n40 = ScalarParameter() {defaultValue=-0.5, group=03_ShakeCrunch_DotsRimHatching, parameterName=ShadowHatchingOffset, sliderMax=0.5, sliderMin=-0.5, sortPriority=20}
  n41 = Subtract(A=Multiply(A=Add(A=@n39, B=@n40) {constB=-0.25}) {constB=0.5, desc=:/}, B=NamedRerouteUsage())
  n42 = NamedRerouteDeclaration(None=SmoothStep(Min=Constant(), Max=Constant() {r=0.1}, Value=Saturate(None=@n41))) {name=ShadowHatchingMask, nodeColor=<rgba 0.3594 0.3594 0.3594 1>}
  n43 = VectorParameter() {
   desc="RGB = Color\r\nA = Alpha / Instensity"
   group=03_ShakeCrunch_DotsRimHatching
   parameterName=ShadowHatchingColor
   sortPriority=10
   defaultValue=<rgba 0.04 0.04 0.04 0.2>
  }
  NamedRerouteDeclaration(None=AppendVector(A=Multiply(A=@n42, B=@n43.RGB), B=@n43.A)) {name=ShadowHatching, nodeColor=<rgba 0.099 0.099 0.099 1>}
  n44 = NamedRerouteUsage()
  n45 = Constant() {desc=Metallic Band Amount, r=3}
  n46 = Multiply(A=Saturate(None=Power(Base=@n44) {constExponent=2.2}), B=@n45)
  n47 = Multiply(A=Saturate(None=ScalarParameter() {defaultValue=0.1, group=03_ShakeCrunch_DotsRimHatching, parameterName=Metallic_Roughness, sliderMax=1})) {constB=2}
  n48 = LinearInterpolate(
   B=Divide(
    A=MaterialFunctionCall!(In=@n46, Ramp Width=@n47) {materialFunction=/Engine/Functions/Engine_MaterialFunctions01/Math/SmoothCeil}.Result
    B=@n45
   )
   Alpha=ScalarParameter() {defaultValue=1, group=03_ShakeCrunch_DotsRimHatching, parameterName=Metallic_Intensity, sliderMax=1}
  ) {constA=1}
  n49 = ScalarParameter() {
   desc="The Bias will adjust the Metallic gradient.\r\n0.0 = the center matches the Color\r\n1.0 = the inner ring matches the Color\r\n2.0 = the outer ring match the Color"
   group=03_ShakeCrunch_DotsRimHatching
   parameterName=Metallic_Bias
   sliderMax=2
   sliderMin=-1
  }
  NamedRerouteDeclaration(None=Add(A=Saturate(None=@n48), B=Multiply(A=@n49) {constB=0.3333})) {name=Metallic_Gradient, nodeColor=<rgba 0.1107 1 0.7992 1>}
  NamedRerouteDeclaration(None=MaterialFunctionCall() {materialFunction=/Game/Global/Materials/Character_Parent_2023/Anime_Parent_2023/MaterialFunctions/MFC_ActorAlighnedScreenPosition}.UVs) {name=ObjectAlighedScreenPosition, nodeColor=<rgba 1 0.867 0 1>}
  n50 = PixelNormalWS()
  NamedRerouteDeclaration(None=DotProduct(A=NamedRerouteDeclaration(None=@n50) {name=NormalWS, nodeColor=<rgba 0.0078 0 1 1>}, B=CameraVectorWS())) {name=NdotV, nodeColor=<rgba 1 0.6242 0 1>}
  NamedRerouteDeclaration(None=MaterialFunctionCall!(Fade Offset=Constant() {r=50}, Fade Distance=Constant() {r=400}) {materialFunction=/Game/Global/Materials/Character_Parent_2023/Anime_Parent_2023/MaterialFunctions/MFC_DistanceFade_FOV_compensation}.Result) {name=DistanceBlending, nodeColor=<rgba 0.4766 1 0 1>}
  n51 = ScalarParameter() {defaultValue=150, group=03_ShakeCrunch_DotsRimHatching, parameterName=HatchingAmount, sliderMax=512, sortPriority=20}
  n52 = Multiply(A=NamedRerouteUsage(), B=LinearInterpolate(A=@n51, B=Multiply(A=@n51) {constB=0.5}, Alpha=NamedRerouteUsage()))
  n53 = Constant() {r=1}
  n54 = Frac(None=ComponentMask(None=Rotator(Coordinate=@n52, Time=@n53) {speed=0.75}) {g=true})
  NamedRerouteDeclaration(None=Abs(None=Add(A=@n54) {constB=-0.5})) {name=HatchingGradient, nodeColor=<rgba 0.3042 0 1 1>}
  n55 = VectorParameter() {group=03_ShakeCrunch_DotsRimHatching, parameterName=RimHatchingVector, sortPriority=10, defaultValue=<rgba 0.8229 0.15 0.4865 1>}
  n56 = DotProduct(
   A=MaterialFunctionCall(Vector=@n55.RGB) {materialFunction=&MF_ProjectToScreenSpace}.Result
   B=NamedRerouteUsage()
  )
  n57 = ScalarParameter() {defaultValue=-0.3, group=03_ShakeCrunch_DotsRimHatching, parameterName=HatchingOffset, sliderMax=0.5, sliderMin=-0.5, sortPriority=20}
  n58 = Subtract(A=Multiply(A=Add(A=@n56, B=@n57) {constB=-0.25}) {constB=0.5, desc=:/}, B=NamedRerouteUsage())
  n59 = NamedRerouteDeclaration(None=SmoothStep(Min=Constant(), Max=Constant() {r=0.1}, Value=Saturate(None=@n58))) {name=RimHatchingMask, nodeColor=<rgba 0.6174 0.2284 0.6354 1>}
  n60 = VectorParameter() {
   desc="RGB = Color\r\nA = Alpha / Instensity"
   group=03_ShakeCrunch_DotsRimHatching
   parameterName=RimHatchingColor
   sortPriority=10
   defaultValue=<rgba 0.04 0.04 0.04 0.2>
  }
  NamedRerouteDeclaration(None=AppendVector(A=Multiply(A=@n59, B=@n60.RGB), B=@n60.A)) {name=RimHatching, nodeColor=<rgba 0.651 0 0.589 1>}
  NamedRerouteDeclaration(None=MaterialFunctionCall!(
    Time_In=ScalarParameter() {defaultValue=-99999, group=03_ShakeCrunch_DotsRimHatching, parameterName=Reactive_Main_CT, sliderMin=-99999}
    AnimDuration=ScalarParameter() {defaultValue=3, group=03_ShakeCrunch_DotsRimHatching, parameterName=Reactive_Main_AnimDuration, sliderMin=0}
   ) {materialFunction=/BRCosmetics/Characters/Player/Male/Medium/Bodies/M_MED_ShakeCrunch/Materials/MF_ShakeCrunch_ReactivityTimeline}.Round) {name=AnimGradient_Reactive, nodeColor=<rgba 1 0 0.3047 1>}
  n61 = NamedRerouteUsage()
  n62 = ScalarParameter() {group=03_ShakeCrunch_DotsRimHatching, parameterName=ColorGlitch_DebugBlend, sliderMax=1}
  n63 = LinearInterpolate(A=NamedRerouteUsage(), B=NamedRerouteUsage(), Alpha=Saturate(None=Add(A=@n61, B=@n62))) {constAlpha=1}
  n64 = VectorParameter() {group=03_ShakeCrunch_DotsRimHatching, parameterName=Wrinkle_Color}
  n65 = Multiply(A=NamedRerouteUsage(), B=ScalarParameter() {group=03_ShakeCrunch_DotsRimHatching, parameterName=Wrinkle_Blend})
  n66 = Multiply(
   A=Add(
    A=NamedRerouteUsage()
    B=MaterialFunctionCall() {materialFunction=/BRCosmetics/Characters/Player/Male/Medium/Bodies/M_MED_ShakeCrunch/Materials/MFC_ShakeCrunch_SteppedTime}.TimeStep
   )
   B=ScalarParameter() {defaultValue=0.5, group=03_ShakeCrunch_DotsRimHatching, parameterName=PaperDetail_Tiling, sliderMax=2}
  )
  n67 = ChannelMaskParameter(None=TextureSample(UVs=Frac(None=@n66)) {texture=/BRCosmetics/Characters/Player/Male/Medium/Bodies/M_MED_ShakeCrunch/Textures/T_ShakeCrunch_Paper}.RGB) {group=03_ShakeCrunch_DotsRimHatching, maskChannel=Green, parameterName=PaperDetail_Channel, defaultValue=<rgba 0 1 0 0>}
  n68 = Add(A=@n67.R, B=ScalarParameter() {group=03_ShakeCrunch_DotsRimHatching, parameterName=PaperDetail_Bias, sliderMax=1, sliderMin=-1})
  n69 = ScalarParameter() {defaultValue=0.8, group=03_ShakeCrunch_DotsRimHatching, parameterName=PaperDetail_Exponent, sliderMax=2}
  n70 = Saturate(None=MaterialFunctionCall(Base=LinearInterpolate(A=@n63, B=@n64.RGB, Alpha=@n65), Blend=Power(Base=@n68, Exp=@n69) {constExponent=0.4545}) {materialFunction=&Blend_Overlay}.Result)
  n71 = MaterialFunctionCall() {materialFunction=&MFC_AnimeShader_LightDirection}
  n72 = PixelNormalWS()
  n73 = Saturate(None=Add(
    A=DotProduct(A=@n71.Direction Vector, B=@n72)
    B=ScalarParameter() {group=03_ShakeCrunch_DotsRimHatching, parameterName=ShadingCurve_Bias, sliderMax=1, sliderMin=-1}
   ))
  n74 = ScalarParameter() {defaultValue=0.02, group=03_ShakeCrunch_DotsRimHatching, parameterName=ShadingCurve_Width, sliderMax=1}
  n75 = ScalarParameter() {defaultValue=1, group=03_ShakeCrunch_DotsRimHatching, parameterName=ShadingCurve_Bands, sliderMax=4, sliderMin=1}
  n76 = LinearInterpolate(
   A=Multiply(A=@n70) {constB=0.66}
   B=@n70
   Alpha=Saturate(None=MaterialFunctionCall!(NdL=@n73, Width=@n74, Bands Amount=@n75) {materialFunction=/Game/Global/Materials/Character_Parent_2023/Anime_Parent_2023/MaterialFunctions/MF_SmoothCelShadingWithBands}.Divide)
  )
  n77 = Multiply(A=@n76, B=NamedRerouteUsage())
  n78 = NamedRerouteUsage()
  n79 = NamedRerouteUsage()
  n80 = LinearInterpolate(
   A=@n79
   B=Multiply(A=@n79, B=NamedRerouteUsage())
   Alpha=ScalarParameter() {group=03_ShakeCrunch_DotsRimHatching, parameterName=SpecularSkinMask_Blending, sliderMax=1}
  )
  n81 = Multiply(A=@n80, B=ScalarParameter() {group=03_ShakeCrunch_DotsRimHatching, parameterName=Specular_Intensity, sliderMax=4})
  n82 = Multiply(A=@n80, B=ScalarParameter() {defaultValue=1, group=03_ShakeCrunch_DotsRimHatching, parameterName=Metallic_SpecularIntensity, sliderMax=4})
  n83 = NamedRerouteUsage()
  NamedRerouteDeclaration(None=Add(A=LinearInterpolate(A=@n76, B=@n77, Alpha=@n78), B=LinearInterpolate(A=@n81, B=@n82, Alpha=@n83))) {name=ColorRaw, nodeColor=<rgba 0 0.5781 1 1>}
  n84 = ScalarParameter() {defaultValue=150, group=03_ShakeCrunch_DotsRimHatching, parameterName=DotsAmount, sliderMax=512, sortPriority=20}
  n85 = Multiply(A=NamedRerouteUsage(), B=LinearInterpolate(A=@n84, B=Multiply(A=@n84) {constB=0.75}, Alpha=NamedRerouteUsage()))
  n86 = Constant() {r=1}
  n87 = Add(A=Frac(None=Rotator(Coordinate=@n85, Time=@n86) {speed=0.75})) {constB=-0.5}
  n88 = Constant()
  NamedRerouteDeclaration(None=Add(A=Distance(A=@n87, B=@n88)) {constB=-0.1, desc=to get full transition}) {name=DotsGradient, nodeColor=<rgba 1 0.0938 0 1>}
  n89 = NamedRerouteUsage()
  n90 = ScalarParameter() {defaultValue=-0.8, group=03_ShakeCrunch_DotsRimHatching, parameterName=HighlightDotsOffset, sliderMax=1, sliderMin=-1, sortPriority=20}
  n91 = ScalarParameter() {defaultValue=1.5, group=03_ShakeCrunch_DotsRimHatching, parameterName=HighlightDotsTransition, sliderMax=2, sortPriority=20}
  n92 = Subtract(A=Multiply(A=Add(A=@n89, B=@n90) {constB=-0.25}, B=@n91) {constB=1.5}, B=NamedRerouteUsage())
  n93 = NamedRerouteDeclaration(None=SmoothStep(Min=Constant(), Max=Constant() {r=0.1}, Value=Saturate(None=@n92))) {name=HighlightDotsMask, nodeColor=<rgba 0.4766 1 0 1>}
  n94 = VectorParameter() {
   desc="RGB = Color\r\nA = Alpha / Instensity"
   group=03_ShakeCrunch_DotsRimHatching
   parameterName=HighlightDotsColor
   sortPriority=10
   defaultValue=<rgba 0.9 0.9 0.9 0.1>
  }
  n95 = NamedRerouteUsage()
  NamedRerouteDeclaration(None=AppendVector(A=Multiply(A=@n93, B=@n94.RGB), B=Multiply(A=@n94.A, B=@n95))) {name=HighlightDots, nodeColor=<rgba 0.9242 1 0.9798 1>}
  n96 = PreSkinnedNormal()
  n97 = DotProduct(A=ShaderStageSwitch(PixelShader=VertexInterpolator(VS=@n96).PS, VertexShader=@n96), B=Constant3Vector() {constant=<rgba 0 1 0 0>})
  n98 = Multiply(A=@n97, B=@n97)
  NamedRerouteDeclaration(None=Multiply(A=@n98, B=@n98) {desc=reduce rim light on the Y axis}) {name=YMask, nodeColor=<rgba 0 1 0.5078 1>}
  NamedRerouteDeclaration(None=ComponentMask(None=FunctionInput() {bUsePreviewValueAsDefault=true, inputName=STT, sortPriority=1}) {g=true}) {name=ThresholdMap, nodeColor=<rgba 0.7109 1 0 1>}
  n99 = GetMaterialAttributes(None=NamedRerouteDeclaration(None=FunctionInput() {bUsePreviewValueAsDefault=true, inputName=MA_In, inputType=FunctionInput_MaterialAttributes}) {name=MA_In, nodeColor=<rgba 1 0.9609 0 1>}) {attributeGetTypes [
    #2
    #6
    #4
    #3
   ]}
  NamedRerouteDeclaration(None=@n99.Metallic) {name=Metallic_In, nodeColor=<rgba 0 0.9581 1 1>}
  NamedRerouteDeclaration(None=@n99.Specular) {name=Specular_In, nodeColor=<rgba 1 0.9289 0 1>}
  n100 = ChannelMaskParameter(None=FunctionInput() {bUsePreviewValueAsDefault=true, inputName=M, sortPriority=1}) {group=03_ShakeCrunch_DotsRimHatching, maskChannel=Blue, parameterName=SkinMask_Channel, defaultValue=<rgba 0 0 1 0>}
  NamedRerouteDeclaration(None=@n100.R) {name=SkinSpec, nodeColor=<rgba 0.9219 0 1 1>}
  n101 = ComponentMask(None=ViewProperty() {property=MEVP_ResolutionFraction}.Property) {r=true}
  NamedRerouteDeclaration(None=Saturate(None=InverseLinearInterpolate(Value=@n101) {constA=0.3})) {name=ResScale, nodeColor=<rgba 0 0.6484 1 1>}
  NamedRerouteDeclaration(None=@n99.BaseColor) {name=BaseColor_In, nodeColor=<rgba 1 0.1346 0 1>}
  n102 = NamedRerouteUsage()
  n103 = NamedRerouteUsage()
  n104 = ComponentMask(None=@n103) {a=true}
  n105 = NamedRerouteUsage()
  n106 = Add(
   A=Multiply(A=ComponentMask(None=@n102) {b=true, g=true, r=true}, B=ComponentMask(None=@n102) {a=true}) {desc=Rim Dots}
   B=Multiply(A=ComponentMask(None=@n103) {b=true, g=true, r=true}, B=Multiply(A=@n104, B=@n105)) {desc=Highlight Dots}
  )
  n107 = NamedRerouteUsage()
  n108 = ComponentMask(None=@n107) {a=true}
  n109 = NamedRerouteUsage()
  n110 = NamedRerouteUsage()
  n111 = ComponentMask(None=@n110) {a=true}
  n112 = NamedRerouteUsage()
  n113 = Add(
   A=Subtract(A=Multiply(A=@n108, B=@n109), B=ComponentMask(None=@n107) {b=true, g=true, r=true}) {desc=Rim Hatching}
   B=Subtract(A=Multiply(A=@n111, B=@n112), B=ComponentMask(None=@n110) {b=true, g=true, r=true}) {desc=Shadow Hatching}
  )
  NamedRerouteDeclaration(None=FunctionOutput(None=Subtract(A=@n106, B=@n113)) {outputName=RimLight_Color, sortPriority=1}) {name=RimLightColor, nodeColor=<rgba 0.2891 1 0 1>}
  NamedRerouteDeclaration(None=@n99.Normal) {name=Normal_In, nodeColor=<rgba 0 0.3061 1 1>}
  NamedRerouteDeclaration(None=DotProduct(
    A=MaterialFunctionCall() {materialFunction=&MFC_AnimeShader_LightDirection}.Direction Vector
    B=NamedRerouteUsage()
   )) {name=LdotN, nodeColor=<rgba 0.2656 0 1 1>}
  n114 = NamedRerouteUsage()
  NamedRerouteDeclaration(None=MaterialFunctionCall!(Color=MaterialFunctionCall!(In=@n114, LUT, LUT Multiplier) {materialFunction=/Game/Packages/Fortress_SharedMaterials/Anime_Master_Material/MF_ApplyLUT}.Result) {materialFunction=/Game/Packages/Fortress_SharedMaterials/Anime_Master_Material/MF_TODColorAdjustment}.Result) {name=ColorFinal, nodeColor=<rgba 1 0.6797 0 1>}
  n115 = VertexColor()
  n116 = Multiply(A=@n115) {constB=2}
  n117 = ScalarParameter() {
   desc="0.0 = surface normals, 1.0 baked vertex color normal"
   group=03_ShakeCrunch_DotsRimHatching
   parameterName=VertexBakedNormal_Blend
   sliderMax=1
  }
  n118 = LinearInterpolate(
   A=Transform(None=NamedRerouteUsage())
   B=Transform(None=Subtract(A=@n116)) {transformSourceType=TRANSFORMSOURCE_Local}
   Alpha=Multiply(A=@n115.A, B=@n117)
  )
  NamedRerouteDeclaration(None=Transform(None=Normalize(VectorInput=@n118)) {transformSourceType=TRANSFORMSOURCE_World, transformType=TRANSFORM_Tangent}) {name=Normal_Out, nodeColor=<rgba 0 1 0.7995 1>}
  NamedRerouteUsage()
  NamedRerouteUsage()
 ]}
