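# Parent character material for the AnvilAlarm Robin "Let" skin, used on skeletal meshes with morph target and mesh deformer support.
# It chains shared 2023 character functions, inserting a skin-specific dots, rim and hatching effect between the PreFX and PostFX stages.
# A cinematic post-FX function ends the chain.
# Material attributes pass between stages, with the shading model selected by expression among lit, subsurface, clear coat and cloth types.
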Material /BRCosmetics/Characters/Player/Male/Medium/Bodies/M_MED_AnvilAlarmRobin/Skins/Let/Materials/M_AnvilAlarm_Let_Parent
graph {
 properties {
  bAutomaticallySetUsageInEditor=false
  bHasPixelAnimation=true
  bUseMaterialAttributes=true
  bUsedWithMeshDeformer=true
  bUsedWithMorphTargets=true
  bUsedWithSkeletalMesh=true
  pixelDepthOffsetMode=PDOM_Legacy
  shadingModel=MSM_FromMaterialExpression
  subsurfaceProfile=&SS_HeroSkin_02
  usedShadingModels=Default Lit | Subsurface | Clear Coat | Subsurface Profile | Cloth
 }
 nodes [
  Constant()
  n1 = MaterialFunctionCall() {materialFunction=&MF_CharacterParent_2023_PreFX}
  n2 = MaterialFunctionCall!(MA_In=@n1.Result, FX=@n1.Result, M=@n1.Result, STT) {materialFunction=&/MF_AnvilAlarm_Let_DotsRimHatching}
  MaterialFunctionCall(In=MaterialFunctionCall(Material=@n2.MA_Out, M=@n1.M, Pre Skinned Local Position=@n1.M, Distance Blend=@n1.M, Fresnel=@n1.M) {materialFunction=&MF_CharacterParent_2023_PostFX}.Result) {materialFunction=&CINE_MF_CharPostFX}
 ]
}
